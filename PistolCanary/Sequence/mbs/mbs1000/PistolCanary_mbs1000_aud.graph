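# Audio level sequence for the PistolCanary mbs1000 sequence, timing music and mix changes.
# It drives a bound PistolCanary_AudioManager with four event triggers, an Event Mix ON toggle and a stepped Music Phase integer, whose changes coincide with the first three events.
# A Music audio track starts just before the first event.
# A MetaSound5.1 track holds a short section at negative frames with listener azimuth and master volume channels.
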LevelSequence /PistolCanary/Sequence/mbs/mbs1000/PistolCanary_mbs1000_aud
bindings [PistolCanary_AudioManager {
  id=#1
  possessed_class=PistolCanary_AudioManager_C
  tracks [
   {
    type=MovieSceneEventTrack
    sections [{
      type=MovieSceneEventTriggerSection
      channels [None {
        type=MovieSceneScriptingEventChannel
        keys [
         [3626, "<Struct 'MovieSceneEvent' (0x852f8f2f640) {}>"]
         [6204, "<Struct 'MovieSceneEvent' (0x852f8f2f5a0) {}>"]
         [6888, "<Struct 'MovieSceneEvent' (0x852f8f2f500) {}>"]
         [8573, "<Struct 'MovieSceneEvent' (0x852f8f2f460) {}>"]
        ]
       }]
     }]
   }
   {
    type=MovieSceneBoolTrack
    display_name=Event Mix ON
    sections [{
      type=MovieSceneBoolSection
      channels [None {type=MovieSceneScriptingBoolChannel, default=true, keys [[3578,false],[3598,true],[9293,true],[9294,false]]}]
     }]
   }
   {
    type=MovieSceneIntegerTrack
    display_name=Music Phase
    sections [{type=MovieSceneIntegerSection, channels [None {type=MovieSceneScriptingIntegerChannel, default=14, keys [[3626,12],[6204,13],[6888,14]]}]}]
   }
  ]
 }]
display_rate {denominator=1, numerator=60}
tick_resolution {denominator=1, numerator=24000}
tracks [
 {
  type=MovieSceneAudioTrack
  display_name=MetaSound5.1
  sections [{
    type=MovieSceneAudioSection
    end_frame=-16841
    start_frame=-16901
    channels [
     Volume {type=MovieSceneScriptingFloatChannel, default=1}
     Pitch {type=MovieSceneScriptingFloatChannel, default=1}
     UE.Listener.Orientation.AzimuthFMovieSceneFloatChannel {type=MovieSceneScriptingFloatChannel, default=117.031}
     MasterVolumeFMovieSceneFloatChannel {type=MovieSceneScriptingFloatChannel, default=1}
     AttachActor {type=MovieSceneScriptingActorReferenceChannel, default="<Struct 'MovieSceneObjectBindingID' (0x852f8ce3980) {}>"}
    ]
   }]
 }
 {
  type=MovieSceneAudioTrack
  display_name=Music
  sections [{
    type=MovieSceneAudioSection
    end_frame=12085
    start_frame=3621
    channels [
     Volume {type=MovieSceneScriptingFloatChannel, default=1}
     Pitch {type=MovieSceneScriptingFloatChannel, default=1}
     AttachActor {type=MovieSceneScriptingActorReferenceChannel, default="<Struct 'MovieSceneObjectBindingID' (0x852f8ce2c20) {}>"}
    ]
   }]
 }
]
